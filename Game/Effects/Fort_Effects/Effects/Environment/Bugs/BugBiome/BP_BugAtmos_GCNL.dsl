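; Bug biome atmosphere actor that applies a looping gameplay cue to actors while they are inside its Bug_Atmosphere component.
; Beginning overlap adds the cue to the other actor and ending overlap removes it, both using the Gameplay Cue Tag and Parameters variables.
; Neither handler filters the overlapping actor, and BeginPlay, Tick, actor overlap and the construction script are empty.
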(event EventBeginPlay)

(event Collision|EventActorBeginOverlap
  (params (OtherActor :type Actor%20Object%20Reference)))

(event EventTick
  (params (DeltaSeconds :type Float%20%28single-precision%29)))

(event OnComponentBeginOverlap(Bug_Atmosphere)
  (params (OverlappedComponent :type Primitive%20Component%20Object%20Reference) (OtherActor :type Actor%20Object%20Reference) (OtherComp :type Primitive%20Component%20Object%20Reference) (OtherBodyIndex :type Integer) (bFromSweep :type Boolean) (SweepResult :type Hit%20Result%20Structure%20%28by%20ref%29))
  (GameplayCue|AddGameplayCueOnActor(Looping) OtherActor (Variables|Default|GetGameplay%20Cue%20Tag) (Variables|Default|GetParameters)))

(event OnComponentEndOverlap(Bug_Atmosphere)
  (params (OverlappedComponent :type Primitive%20Component%20Object%20Reference) (OtherActor :type Actor%20Object%20Reference) (OtherComp :type Primitive%20Component%20Object%20Reference) (OtherBodyIndex :type Integer))
  (GameplayCue|RemoveGameplayCueOnActor(Looping) OtherActor (Variables|Default|GetGameplay%20Cue%20Tag) (Variables|Default|GetParameters)))

(fn ConstructionScript)
