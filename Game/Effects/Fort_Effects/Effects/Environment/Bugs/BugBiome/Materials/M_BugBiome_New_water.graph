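# Unlit, masked, two-sided material for animated water splash sprites in the bug biome, simulating falling droplets in the shader.
# Time plus vertex-color green gives each sprite a looping normalized age and a downward, outward velocity.
# The Sprite function billboards quads for world position offset, with sphere-mask scaling gated by vertex alpha.
# A randomly rotated SubUV splash texture feeds a masked Bayer dither, while time-of-day functions tint the color.
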Material /Game/Effects/Fort_Effects/Effects/Environment/Bugs/BugBiome/Materials/M_BugBiome_New_water
graph {
 properties {bAutomaticallySetUsageInEditor=false, blendMode=BLEND_Masked, shadingModel=MSM_Unlit, twoSided=true, usedShadingModels=Unlit}
 nodes [
  n1 = LinearInterpolate(
   B=ScalarParameter() {defaultValue=500}
   Alpha=SphereMask(
    A=NamedRerouteUsage()
    B=ScalarParameter() {defaultValue=1, parameterName=Param_2}
    Radius=ScalarParameter() {defaultValue=1.2833, parameterName=Param_3}
    Hardness=ScalarParameter() {parameterName=Param_4}
   )
  )
  n2 = Step(Y=VertexColor().A, X=ScalarParameter() {defaultValue=0.25, parameterName=Param_5})
  Add(
   A=MaterialFunctionCall!(Center Location=WorldPosition().XYZ, 0-1 UVs=TextureCoordinate(), XY scale=Multiply(A=@n1, B=@n2), OverridePivot, "Normals (Optional)") {materialFunction=&Sprite}.World position offset
   B=NamedRerouteUsage()
  )
  Add()
  n3 = Constant() {r=2}
  n4 = VertexColor()
  n5 = Multiply(A=Time(), B=LinearInterpolate(B=ScalarParameter() {defaultValue=5, parameterName=Param_11}, Alpha=@n4.G) {constA=-5})
  n6 = LinearInterpolate(Alpha=@n4.G) {constB=360}
  n7 = MaterialFunctionCall!(
   U_Images=@n3
   V_Images=@n3
   SubUV Texture RGB
   Time Speed=LinearInterpolate(Alpha=VertexColor().G) {constB=16}
   Manual Time Control=StaticBoolParameter() {defaultValue=true, parameterName="Manual Time Control (SB)"}
   Time
   UVs=Rotator(Coordinate=TextureCoordinate(), Time=Add(A=@n5, B=@n6))
  ) {materialFunction=/Game/Effects/Fort_Effects/Materials/_Functions/MF_SubUV}
  n8 = NamedRerouteUsage()
  n9 = ScalarParameter() {defaultValue=0.5, parameterName=Param_8}
  n10 = ScalarParameter() {defaultValue=0.5, parameterName=Param_9}
  n11 = ScalarParameter() {defaultValue=0.25, parameterName=Param_10}
  MaterialFunctionCall!(
   In=Multiply(
    A=TextureSample(UVs=@n7.UVs) {samplerType=SAMPLERTYPE_LinearColor, texture=/Game/Effects/Fort_Effects/Textures/Water/T_waterSplash_D}.G
    B=SphereMask(A=@n8, B=@n9, Radius=@n10, Hardness=@n11)
   )
   Iterations
   Scale
  ) {materialFunction=/Game/Packages/Fortress_SharedMaterials/Base_Material_Functions/MF_DitherBayerNoTex}
  Multiply(
   A=VectorParameter() {parameterName=Param_1, defaultValue=<rgba 0.7344 0.7344 0.7344 0>}.RGB
   B=MaterialFunctionCall(
    Color_Morning=VectorParameter() {parameterName=Param_3, defaultValue=<rgba 0.6979 0.6979 0.6979 0>}.RGB
    Color_Day=VectorParameter() {parameterName=Param_3, defaultValue=<rgba 0.6979 0.6979 0.6979 0>}.RGB
    Color_Evening=VectorParameter() {parameterName=Param_3, defaultValue=<rgba 0.6979 0.6979 0.6979 0>}.RGB
    Color_Night=VectorParameter() {parameterName=Param_2, defaultValue=<rgba 0.1238 0.1557 0.4323 1>}.RGB
   ) {materialFunction=&Color_Time_of_Day}.Result
  )
  n12 = VertexColor()
  n13 = Multiply(A=Time(), B=ScalarParameter() {defaultValue=0.75, parameterName=Param_6})
  n14 = ScalarParameter() {defaultValue=1, parameterName=Param_1}
  n15 = Divide(A=Fmod(A=Add(A=@n12.G, B=@n13), B=@n14), B=@n14)
  NamedRerouteDeclaration(None=@n15) {name=Normalized Age, nodeColor=<rgba 1 0.3281 0 1>}
  n16 = VectorParameter() {defaultValue=<rgba 0 0 -750 0>}
  n17 = VertexColor()
  n18 = ScalarParameter() {defaultValue=1, parameterName=2}
  n19 = Divide(A=Fmod(A=Add(A=@n17.G, B=@n13), B=@n18), B=@n18)
  n20 = VectorParameter() {parameterName=Outward Velocity, defaultValue=<rgba -50 -125 0 0>}
  NamedRerouteDeclaration(None=Add(A=Multiply(A=@n15, B=@n16.RGB), B=Multiply(A=@n19, B=@n20.RGB))) {name=Velocity, nodeColor=<rgba 0.0078 0 1 1>}
  ScalarParameter() {parameterName=Param_7}
  TextureCoordinate() {coordinateIndex=1}
  Time()
 ]
}
