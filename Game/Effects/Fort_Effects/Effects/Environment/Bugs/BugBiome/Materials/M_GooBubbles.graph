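# M_GooBubbles is a masked, default-lit particle material for goo bubbles in the BugBiome effects, enabled for Niagara sprites, ribbons and mesh particles.
# A Fresnel term feeds the MF_FX_EmissiveAdjustments function as its emissive input.
# A lerp takes particle alpha as its A input and is driven by the green channel of a borrowed character electric-noise texture.
# The graph shows no links to the material outputs.
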Material /Game/Effects/Fort_Effects/Effects/Environment/Bugs/BugBiome/Materials/M_GooBubbles
graph {
 properties {
  bAutomaticallySetUsageInEditor=false
  bUsedWithNiagaraMeshParticles=true
  bUsedWithNiagaraRibbons=true
  bUsedWithNiagaraSprites=true
  blendMode=BLEND_Masked
  usedShadingModels=Default Lit
 }
 nodes [
  LinearInterpolate(
   A=ParticleColor().A
   Alpha=TextureSample() {samplerType=SAMPLERTYPE_LinearColor, texture=/Game/Characters/Player/Male/Medium/Bodies/M_MED_SRDrift/Textures/T_ElectricNoise}.G
  )
  MaterialFunctionCall(Emissive=Fresnel()) {materialFunction=&MF_FX_EmissiveAdjustments}
 ]
}
